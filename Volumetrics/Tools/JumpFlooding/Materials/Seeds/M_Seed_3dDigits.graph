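# Unlit seed material in the volumetric Jump Flooding tools, built to render debug digits, apparently as 3D test seeds.
# One branch tiles UVs into a 16-cell grid, prints each cell index with DebugFloat2Values and floors the color-coded result.
# Another prints EncodeVolumeCoordinates output, driven by the XYFrames flipbook layout parameter, through DebugFloat3Values.
# Neither branch shows a link to a material output, and a 1Dto2DIndex node sits unconnected.
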Material /Volumetrics/Tools/JumpFlooding/Materials/Seeds/M_Seed_3dDigits
graph {
 properties {pixelDepthOffsetMode=PDOM_Legacy, shadingModel=MSM_Unlit, usedShadingModels=Unlit}
 nodes [
  n1 = Multiply(A=TextureCoordinate()) {constB=16}
  Floor(None=MaterialFunctionCall!(Vector2=Floor(None=@n1), MaximumNumberOfDigits, UVs=Frac(None=@n1), DebugTextLocation RG_UpperRight BA_LowerLeft, Component Spacing) {materialFunction=/Engine/Functions/Engine_MaterialFunctions02/Utility/DebugFloat2Values}.ColorCodedOutput)
  n2 = VectorParameter() {parameterName=XYFrames, defaultValue=<rgba 16 16 0 0>}
  MaterialFunctionCall!(
   Vector3=MaterialFunctionCall!(UVs, XY Frames=ComponentMask(None=@n2.RGB) {g=true, r=true}) {materialFunction=/Volumetrics/Content/VolumeTextures/MaterialFunctions/EncodeVolumeCoordinates}.Result
   MaximumNumberOfDigits
   UVs
   DebugTextLocation RG_UpperRight BA_LowerLeft
   Component Spacing
  ) {materialFunction=/Engine/Functions/Engine_MaterialFunctions02/Utility/DebugFloat3Values}
  MaterialFunctionCall!(1D Index, Cells XY) {materialFunction=/Engine/Functions/Engine_MaterialFunctions01/Coordinates/1Dto2DIndex}
 ]
}
