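# Unlit seed material, by name for 3D voxel jump flooding in the Volumetrics tools.
# A hard-edged SphereMask over the UVs marks a tiny dot at the texture center, with its radius set by a Divide node.
# The mask is multiplied by a constant vector holding the center coordinate in red and green, with blue at zero.
# A loose constant and two black EXR texture samples feed none of the shown nodes.
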Material /Volumetrics/Tools/JumpFlooding/Materials/Seeds/M_Seed_3dVoxels
graph {
 properties {pixelDepthOffsetMode=PDOM_Legacy, shadingModel=MSM_Unlit, usedShadingModels=Unlit}
 nodes [
  Constant() {r=0.01}
  Multiply(
   A=SphereMask(A=TextureCoordinate(), B=Constant() {r=0.5}, Radius=Divide() {constB=640}, Hardness=Constant() {r=1})
   B=Constant3Vector() {constant=<rgba 0.5 0.5 0 0>}
  )
  TextureSample() {samplerType=SAMPLERTYPE_LinearColor, texture=/Engine/Functions/Engine_MaterialFunctions02/PivotPainter2/Black_1x1_EXR_Texture}
  TextureSample() {samplerType=SAMPLERTYPE_LinearColor, texture=&Black_1x1_EXR_Texture}
 ]
}
